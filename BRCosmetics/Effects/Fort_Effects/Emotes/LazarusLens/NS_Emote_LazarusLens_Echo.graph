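# Niagara mesh effect for the LazarusLens emote's echo, combining a translucent cone with rising ripples tinted by time of day.
# Echo_Cone bursts one mesh that stretches vertically, swells and collapses in width, and sets emitter WPO top/bottom values from loop-age curves.
# MeshRipples continuously spawns upward-moving meshes that inherit source movement, animate material parameters and fade with particle and system age.
# A second ripple emitter, MeshRipples001, is disabled.
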NiagaraSystem /BRCosmetics/Effects/Fort_Effects/Emotes/LazarusLens/NS_Emote_LazarusLens_Echo
dependencies {
 usedDataInterfaces [&NiagaraDataInterfaceVectorCurve, &NiagaraDataInterfaceCurve]
 usedDynamicInputs []
 usedModules [&SystemState]
}
emitters {
 Echo_Cone {
  emitterUpdateScript [
   &EmitterState()
   &SpawnBurst_Instantaneous(Use Spawn Probability=false, UseLoopCountLimit=true)
   /SetVariables(
    Emitter.WPO Bottom:NiagaraFloat=&FloatFromCurve(
     FloatCurve=DataInterface() {propertyValues {Curve {keys [
         {interpMode=RCIM_Cubic, tangentMode=RCTM_User, time=0, value=1}
         {arriveTangent=0.9131, interpMode=RCIM_Cubic, leaveTangent=0.9131, time=0.2374, value=1.7094}
         {arriveTangent=0.8622, interpMode=RCIM_Cubic, leaveTangent=0.8622, tangentMode=RCTM_User, time=0.6945, value=1.6341}
         {interpMode=RCIM_Cubic, leaveTangent=0.5501, time=1, value=1}
        ]}}}
     CurveIndex=$Emitter.NormalizedLoopAge
     Scale Curve=5
    )
    Emitter.WPO Top:NiagaraFloat=&FloatFromCurve(
     FloatCurve=DataInterface() {propertyValues {Curve {keys [
         {arriveTangent=-1.7296, interpMode=RCIM_Cubic, leaveTangent=-1.7296, tangentMode=RCTM_User, time=0, value=1}
         {arriveTangent=0.0788, interpMode=RCIM_Cubic, leaveTangent=0.0788, time=0.3332, value=0.6599}
         {arriveTangent=0.8622, interpMode=RCIM_Cubic, leaveTangent=0.8622, tangentMode=RCTM_User, time=0.7365, value=1.0581}
         {arriveTangent=-2.7071, interpMode=RCIM_Cubic, leaveTangent=-2.7071, tangentMode=RCTM_User, time=1, value=1}
        ]}}}
     CurveIndex=$Emitter.NormalizedLoopAge
     Scale Curve=30
    )
   ) {isSetParameters=true}
  ]
  particleSpawnScript [&InitializeParticle(
    Color=&FNTimeOfDayColor(
     Night Color=<rgba 0.3332 2 1.9801 0.075>
     Morning Color=<rgba 0.3332 2 1.9801 0.3>
     Day Color=<rgba 0.3332 2 1.9801 0.3>
     Evening Color=<rgba 0.3332 2 1.9801 0.3>
    )
    Lifetime=1.75
    Mesh Scale=<0.75 0.75 0.425>
    Mesh Scale Mode=::NewEnumerator4
    Position Offset=<0 0 7.5>
    UsePositionOffset=true
    Write Material Random=true
   )]
  particleUpdateScript [
   &ParticleState()
   &ScaleColor(
    Scale Alpha=&FloatFromCurve(FloatCurve=DataInterface() {propertyValues {Curve {keys [
         {arriveTangent=4.923, interpMode=RCIM_Cubic, leaveTangent=4.923, tangentMode=RCTM_User, time=0, value=0}
         {arriveTangent=1.1111, leaveTangent=1.1111, time=0.4, value=1}
         {arriveTangent=0.1816, interpMode=RCIM_Cubic, leaveTangent=0.1816, tangentMode=RCTM_User, time=0.7, value=1}
         {arriveTangent=-7.3064, interpMode=RCIM_Cubic, leaveTangent=-7.3064, tangentMode=RCTM_User, time=1, value=0}
        ]}}})
    ScaleA=true
    ScaleRGB=true
   )
   &ScaleMeshSize(Scale Factor=&MakeVector(
     X=&FloatFromCurve(FloatCurve=DataInterface() {propertyValues {Curve {keys [
          <at 0 0.35 RCIM_Cubic>
          {arriveTangent=-0.35, interpMode=RCIM_Cubic, leaveTangent=-0.35, time=0.5, value=1}
          {arriveTangent=-3.2143, interpMode=RCIM_Cubic, leaveTangent=-3.2143, tangentMode=RCTM_User, time=1, value=0}
         ]}}})
     Y=&FloatFromCurve(FloatCurve=DataInterface() {propertyValues {Curve {keys [
          <at 0 0.35 RCIM_Cubic>
          {arriveTangent=-0.35, interpMode=RCIM_Cubic, leaveTangent=-0.35, time=0.5, value=1}
          {arriveTangent=-3.2143, interpMode=RCIM_Cubic, leaveTangent=-3.2143, tangentMode=RCTM_User, time=1, value=0}
         ]}}})
     Z=&FloatFromCurve(
      FloatCurve=DataInterface() {propertyValues {Curve {keys [
          {arriveTangent=1.7984, interpMode=RCIM_Cubic, leaveTangent=1.7984, tangentMode=RCTM_User, time=0, value=0}
          {arriveTangent=0.0224, interpMode=RCIM_Cubic, leaveTangent=0.0224, tangentMode=RCTM_User, time=1, value=1}
         ]}}}
      Scale Curve=1.25
     )
    ))
  ]
  rendererClasses [&NiagaraMeshRendererProperties]
  renderers [{rendererClass=&NiagaraMeshRendererProperties, rendererIndex=0}]
 }
 MeshRipples {
  emitterUpdateScript [
   &EmitterState()
   &SpawnRate(SpawnRate=7, Use Spawn Probability=false)
  ]
  particleSpawnScript [
   &InitializeParticle(
    Color=&FNTimeOfDayColor(
     Night Color=<rgba 0.0972 0.8106 0.8714 0.025>
     Morning Color=<rgba 0.0972 0.8106 0.8714 0.05>
     Day Color=<rgba 0.0972 0.8106 0.8714 0.04>
     Evening Color=<rgba 0.0972 0.8106 0.8714 0.05>
    )
    Lifetime Max=0.65
    Lifetime Min=0.6
    Lifetime Mode=::NewEnumerator1
    Mesh Scale Mode=::NewEnumerator1
    Mesh Uniform Scale Max=0.65
    Mesh Uniform Scale Min=0.15
    Position Offset=<0 0 5>
    UsePositionOffset=true
    Write Material Random=true
   )
   &AddVelocity(Velocity=&RandomRangeVector(Minimum=<0 0 55>, Maximum=<0 0 60>))
   &InitialMeshOrientation(
    Mesh Orientation Mode=::NewEnumerator1
    Rotation=&RandomRangeVector(Minimum=<0 0 0>, Maximum=<0 0 1>)
    Use Rotation Vector=true
   )
  ]
  particleUpdateScript [
   &ParticleState()
   &Drag(Drag=0, Use Linear Drag=true, Use Rotational Drag=true)
   &InheritSourceMovement(
    Applied Position Scale=&VectorFromCurve(VectorCurve=DataInterface() {propertyValues {
       XCurve {keys [<at 0 1 RCIM_Cubic>, {arriveTangent=-0.2519, interpMode=RCIM_Cubic, leaveTangent=-0.2519, tangentMode=RCTM_User, time=1, value=0}]}
       YCurve {keys [<at 0 1 RCIM_Cubic>, {arriveTangent=-0.2519, interpMode=RCIM_Cubic, leaveTangent=-0.2519, tangentMode=RCTM_User, time=1, value=0}]}
       ZCurve {keys [<at 0 1 RCIM_Cubic>, {arriveTangent=-0.2519, interpMode=RCIM_Cubic, leaveTangent=-0.2519, tangentMode=RCTM_User, time=1, value=0}]}
      }})
    ApplyForce=false
    ApplyPosition=true
    ApplyPositionFalloff=false
   )
   &SolveForcesAndVelocity(Clamp Velocity=false, Limit Acceleration=false)
   &UpdateMeshOrientation(Rotation Vector=<0 0 1>)
   &ScaleMeshSize(Scale Factor=&VectorFromFloat(Value=&FloatFromCurve(
      FloatCurve=DataInterface() {propertyValues {Curve {keys [
          {interpMode=RCIM_Cubic, tangentMode=RCTM_User, time=0, value=0}
          {
           arriveTangent=0.6713
           arriveTangentWeight=0.2432
           interpMode=RCIM_Cubic
           leaveTangent=-0.5605
           leaveTangentWeight=0.2623
           tangentMode=RCTM_Break
           tangentWeightMode=RCTWM_WeightedBoth
           time=0.2821
           value=0.3697
          }
          {
           arriveTangent=-1.0936
           arriveTangentWeight=0.2519
           interpMode=RCIM_Cubic
           leaveTangent=-1.0936
           leaveTangentWeight=0.2519
           tangentMode=RCTM_User
           tangentWeightMode=RCTWM_WeightedBoth
           time=0.6874
           value=0.5468
          }
          {arriveTangent=0.3822, interpMode=RCIM_Cubic, leaveTangent=0.3822, tangentMode=RCTM_User, time=1, value=1}
         ]}}}
      Scale Curve=&RandomRangeFloat(Minimum=0.95, Maximum=1.05)
     )))
   &DynamicMaterialParameters(
    Index 0 Param 1=&FloatFromCurve(FloatCurve=DataInterface() {propertyValues {Curve {keys [<at 0.1 5 RCIM_Cubic>,<at 0.95 0 RCIM_Cubic>]}}})
    Index 0 Param 2=&FloatFromCurve(FloatCurve=DataInterface() {propertyValues {Curve {keys [<at 0 5 RCIM_Cubic>,<at 0.4 36 RCIM_Cubic>]}}})
    Index 0 Param 3=1
    Index 0 Param 4=&FloatFromCurve(FloatCurve=DataInterface() {propertyValues {Curve {keys [<at 0 1 RCIM_Cubic>,<at 1 0 RCIM_Cubic>]}}})
    Module.Param0WriteEnabled=true
    Module.Param1WriteEnabled=true
    Module.Param2WriteEnabled=true
    Module.Param3WriteEnabled=true
   )
   &ScaleColor(
    Scale Alpha=&FloatFromCurve(
     FloatCurve=DataInterface() {propertyValues {Curve {keys [
         {arriveTangent=9.6668, interpMode=RCIM_Cubic, leaveTangent=9.6668, tangentMode=RCTM_User, time=0, value=0}
         <at 0.025 1 RCIM_Cubic>
         {arriveTangent=-2.5941, interpMode=RCIM_Cubic, leaveTangent=-2.5941, tangentMode=RCTM_User, time=1, value=0}
        ]}}}
     Scale Curve=&FloatFromCurve(
      FloatCurve=DataInterface() {propertyValues {Curve {keys [
          {tangentMode=RCTM_User, time=0, value=1}
          {arriveTangent=-0.0405, interpMode=RCIM_Cubic, leaveTangent=-0.0405, tangentMode=RCTM_User, time=1, value=1}
          {arriveTangent=-2.1589, interpMode=RCIM_Cubic, leaveTangent=-2.1589, tangentMode=RCTM_User, time=2, value=0}
         ]}}}
      CurveIndex=$Engine.System.Age
     )
    )
    ScaleA=true
    ScaleRGB=true
   )
  ]
  rendererClasses [&NiagaraMeshRendererProperties]
  renderers [{rendererClass=&NiagaraMeshRendererProperties, rendererIndex=0}]
 }
 MeshRipples001 {
  bEnabled=false
  emitterUpdateScript [
   &EmitterState()
   &SpawnRate(SpawnRate=10, Use Spawn Probability=false)
  ]
  particleSpawnScript [
   &InitializeParticle(
    Color=<rgba 0.0626 0.7686 1 0.07>
    Lifetime Max=0.45
    Lifetime Min=0.4
    Lifetime Mode=::NewEnumerator1
    Mesh Scale Mode=::NewEnumerator1
    Mesh Uniform Scale Max=65
    Mesh Uniform Scale Min=60
    Position Offset=<0 0 2.5>
    UsePositionOffset=true
    Write Material Random=true
   )
   &AddVelocity(Velocity=&RandomRangeVector(Minimum=<0 0 100>, Maximum=<0 0 120>))
  ]
  particleUpdateScript [
   &ParticleState()
   &Drag(Use Linear Drag=true, Use Rotational Drag=true)
   &InheritSourceMovement(
    Applied Position Scale=&VectorFromCurve(VectorCurve=DataInterface() {propertyValues {
       XCurve {keys [<at 0 1 RCIM_Cubic>, {arriveTangent=-0.2519, interpMode=RCIM_Cubic, leaveTangent=-0.2519, tangentMode=RCTM_User, time=1, value=0}]}
       YCurve {keys [<at 0 1 RCIM_Cubic>, {arriveTangent=-0.2519, interpMode=RCIM_Cubic, leaveTangent=-0.2519, tangentMode=RCTM_User, time=1, value=0}]}
       ZCurve {keys [<at 0 1 RCIM_Cubic>, {arriveTangent=-0.2519, interpMode=RCIM_Cubic, leaveTangent=-0.2519, tangentMode=RCTM_User, time=1, value=0}]}
      }})
    ApplyForce=false
    ApplyPosition=true
    ApplyPositionFalloff=false
   )
   &SolveForcesAndVelocity(Clamp Velocity=false, Limit Acceleration=false)
   &UpdateMeshOrientation(Rotation Rate=0.2, Rotation Vector=<0 0 1>)
   &ScaleMeshSize(Scale Factor=&VectorFromFloat(Value=&FloatFromCurve(
      FloatCurve=DataInterface() {propertyValues {Curve {keys [
          {interpMode=RCIM_Cubic, tangentMode=RCTM_User, time=0, value=0.025}
          {arriveTangent=0.9343, interpMode=RCIM_Cubic, leaveTangent=0.9343, time=0.2988, value=0.4734}
          {arriveTangent=-0.2114, interpMode=RCIM_Cubic, leaveTangent=-0.2114, tangentMode=RCTM_User, time=0.7161, value=0.6941}
          {interpMode=RCIM_Cubic, leaveTangent=2.7348, time=1, value=1}
         ]}}}
      Scale Curve=&RandomRangeFloat(Minimum=0.95, Maximum=1.05)
     )))
   &ScaleColor(
    Scale Alpha=&FloatFromCurve(FloatCurve=DataInterface() {propertyValues {Curve {keys [
         {arriveTangent=9.6668, interpMode=RCIM_Cubic, leaveTangent=9.6668, tangentMode=RCTM_User, time=0, value=0}
         <at 0.025 1 RCIM_Cubic>
         {arriveTangent=-3.4206, interpMode=RCIM_Cubic, leaveTangent=-3.4206, tangentMode=RCTM_User, time=1, value=0}
        ]}}})
    ScaleA=true
    ScaleRGB=true
   )
   &DynamicMaterialParameters(
    Index 0 Param 1=1
    Index 0 Param 2=1
    Index 0 Param 3=1
    Index 0 Param 4=1
    Module.Param0WriteEnabled=true
    Module.Param1WriteEnabled=true
    Module.Param2WriteEnabled=true
    Module.Param3WriteEnabled=true
   )
  ]
  rendererClasses [&NiagaraMeshRendererProperties]
  renderers [{rendererClass=&NiagaraMeshRendererProperties, rendererIndex=0}]
 }
}
summary {
 systemName=NS_Emote_LazarusLens_Echo
 emitters [
  {emitterName=MeshRipples, rendererClasses [&NiagaraMeshRendererProperties]}
  {bEnabled=false, emitterName=MeshRipples001, rendererClasses [&NiagaraMeshRendererProperties]}
  {emitterName=Echo_Cone, rendererClasses [&NiagaraMeshRendererProperties]}
 ]
 userVariables [
  User.bIsFrontendPreview {type=NiagaraBool}
  User.Intensity {type=NiagaraFloat, defaultValue=0, description=""}
 ]
}
