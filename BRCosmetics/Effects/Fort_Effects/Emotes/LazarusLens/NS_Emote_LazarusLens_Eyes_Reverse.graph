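# Niagara system for the LazarusLens emote (Reverse variant) that draws red glowing sprites over a character's eyes.
# Two enabled emitters burst a pair of camera-offset sprites, pinned per partition to User.SkeletalMesh bones or sockets, with mirrored offsets blended by System.Style.
# Curves flash brightness then fade it, with alpha modulated by time of day and System.Blink.
# Two similar glow emitters are disabled.
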NiagaraSystem /BRCosmetics/Effects/Fort_Effects/Emotes/LazarusLens/NS_Emote_LazarusLens_Eyes_Reverse
dependencies {
 usedDataInterfaces [&NiagaraDataInterfaceCurve, &NiagaraDataInterfaceVector2DCurve]
 usedDynamicInputs [
  &ClampFloat
  &Subtract_Float
  &MakeFloatFromVector
  &RotateVectorByQuaternion
  &Add_Float
 ]
 usedModules [
  &SystemState
  &SampleSkeletalMesh
  &self:SystemSpawnScript.SystemScriptSource.SystemScriptGraph.NiagaraNodeAssignment_2.SetVariables_#1
  &self:SystemSpawnScript.SystemScriptSource.SystemScriptGraph.NiagaraNodeAssignment_1.SetVariables_#2
 ]
}
emitters {
 Glow_Eyes {
  bEnabled=false
  emitterUpdateScript [
   &EmitterState()
   &SpawnBurst_Instantaneous(Use Spawn Probability=false, UseLoopCountLimit=true)
  ]
  particleSpawnScript [
   &PartitionParticles()
   &InitializeParticle(
    Color=<rgba 20 0 0 1>
    Lifetime=1.15
    Sprite Size Mode=::NewEnumerator3
    Uniform Sprite Size=8
    UsePositionOffset=false
    Write Material Random=true
   )
   &CameraOffset(Camera Offset Amount=2.75)
  ]
  particleUpdateScript [
   &ParticleState()
   &SkeletalMeshLocation(
    Bone / Socket Index=$Particles.PartitionParticles.Partition
    Bone Sampling Mode=::NewEnumerator4
    MaskSpawnGroup=false
    Sampled Position Offset=&Lerp_Vector(A=<0.5 0.25 1>, B=<-0.5 0.25 1>, Alpha=&MakeFloatFromInt(Int=$Particles.PartitionParticles.Partition))
    Skeletal Mesh=$User.SkeletalMesh
   )
   &ScaleColor(
    Scale Alpha=&FloatFromCurve(
     FloatCurve=DataInterface() {propertyValues {Curve {keys [
         {arriveTangent=-2.5249, interpMode=RCIM_Cubic, leaveTangent=-2.5249, tangentMode=RCTM_User, time=0, value=1}
         {arriveTangent=-0.0405, interpMode=RCIM_Cubic, leaveTangent=-0.0405, tangentMode=RCTM_User, time=1, value=0}
        ]}}}
     Scale Curve=&Multiply_Float(
      A=&FNTimeOfDayScalar(Night Scale=0.7, Morning Scale=0.8, Evening Scale=0.8)
      B=&MakeCustomFloatFromBool(Bool=&SetBoolByFloatComparison(A=$System.Blink, B=0.4), True Float=$System.Blink, False Float=1)
     )
    )
    ScaleA=true
    ScaleRGB=false
   )
  ]
  rendererClasses [&NiagaraSpriteRendererProperties]
  renderers [{rendererClass=&NiagaraSpriteRendererProperties}]
 }
 Glow_Eyes001 {
  bEnabled=false
  emitterUpdateScript [
   &EmitterState()
   &SpawnBurst_Instantaneous(Use Spawn Probability=false, UseLoopCountLimit=true)
  ]
  particleSpawnScript [
   &PartitionParticles()
   &InitializeParticle(
    Color=<rgba 10 0 0 0.5>
    Lifetime=1.15
    Mesh Scale Mode=::NewEnumerator3
    Mesh Uniform Scale=0.5
    Sprite Size=<20 8>
    Sprite Size Mode=::NewEnumerator4
    UsePositionOffset=false
    Write Material Random=true
   )
   &CameraOffset(Camera Offset Amount=2.6)
  ]
  particleUpdateScript [
   &ParticleState()
   &SkeletalMeshLocation(
    Bone / Socket Index=$Particles.PartitionParticles.Partition
    Bone Sampling Mode=::NewEnumerator4
    MaskSpawnGroup=false
    Sampled Position Offset=&Lerp_Vector(A=<0.5 0.25 1>, B=<-0.5 0.25 1>, Alpha=&MakeFloatFromInt(Int=$Particles.PartitionParticles.Partition))
    Skeletal Mesh=$User.SkeletalMesh
   )
   &ScaleColor(
    Scale Alpha=&FloatFromCurve(
     FloatCurve=DataInterface() {propertyValues {Curve {keys [
         {arriveTangent=9.0725, interpMode=RCIM_Cubic, leaveTangent=9.0725, tangentMode=RCTM_User, time=0, value=0}
         {arriveTangent=-2.5249, interpMode=RCIM_Cubic, leaveTangent=-2.5249, tangentMode=RCTM_User, time=0.1408, value=0.9847}
         {arriveTangent=-0.0405, interpMode=RCIM_Cubic, leaveTangent=-0.0405, tangentMode=RCTM_User, time=1, value=0}
        ]}}}
     Scale Curve=&Multiply_Float(
      A=&FNTimeOfDayScalar(Night Scale=0.7, Morning Scale=0.8, Evening Scale=0.8)
      B=&MakeCustomFloatFromBool(Bool=&SetBoolByFloatComparison(A=$System.Blink, B=0.4), True Float=$System.Blink, False Float=1)
     )
    )
    ScaleA=true
    ScaleRGB=true
   )
   &ScaleSpriteSize(
    Non-Uniform Curve Sprite Scale=DataInterface() {propertyValues {
      XCurve {keys [<at 0 1>, {arriveTangent=1.703, interpMode=RCIM_Cubic, leaveTangent=1.703, tangentMode=RCTM_User, time=1, value=2}]}
      YCurve {keys [
        {arriveTangent=9.0725, interpMode=RCIM_Cubic, leaveTangent=9.0725, tangentMode=RCTM_User, time=0, value=0.5}
        {arriveTangent=-0.7034, interpMode=RCIM_Cubic, leaveTangent=-0.7034, tangentMode=RCTM_User, time=0.1, value=1.25}
        <at 1 0.75>
       ]}
     }}
    Scale Sprite Size Mode=::NewEnumerator3
   )
  ]
  rendererClasses [&NiagaraSpriteRendererProperties]
  renderers [{rendererClass=&NiagaraSpriteRendererProperties}]
 }
 Glow_Eyes002 {
  emitterUpdateScript [
   &EmitterState()
   &SpawnBurst_Instantaneous(Spawn Count=2, Use Spawn Probability=false, UseLoopCountLimit=false)
  ]
  particleSpawnScript [
   &PartitionParticles()
   &InitializeParticle(
    Color=<rgba 4 0 0 1>
    Mesh Scale Mode=::NewEnumerator3
    Mesh Uniform Scale=0.5
    Sprite Size=<16 6.5>
    Sprite Size Mode=::NewEnumerator4
    UsePositionOffset=false
    Write Material Random=true
   )
   &CameraOffset(Camera Offset Amount=2.75)
  ]
  particleUpdateScript [
   &ParticleState()
   &SkeletalMeshLocation(
    Bone / Socket Index=$Particles.PartitionParticles.Partition
    Bone Sampling Mode=::NewEnumerator4
    MaskSpawnGroup=false
    Sampled Position Offset=&Lerp_Vector(
     A=&Lerp_Vector(
      A=<0.375 0.25 1>
      B=<-0.4 0.25 1>
      Alpha=&MakeFloatFromInt(Int=$Particles.PartitionParticles.Partition)
     )
     B=&Lerp_Vector(
      A=<1 1.25 1.25>
      B=<-1 1.25 1.25>
      Alpha=&MakeFloatFromInt(Int=$Particles.PartitionParticles.Partition)
     )
     Alpha=$System.Style
    )
    Skeletal Mesh=$User.SkeletalMesh
   )
   &ScaleColor(
    Scale Alpha=&FloatFromCurve(
     FloatCurve=DataInterface() {propertyValues {Curve {keys [
         {
          arriveTangent=9.0725
          interpMode=RCIM_Cubic
          leaveTangent=9.0725
          leaveTangentWeight=0.1521
          tangentMode=RCTM_User
          tangentWeightMode=RCTWM_WeightedBoth
          time=0
          value=0
         }
         {
          arriveTangent=-0.9637
          arriveTangentWeight=0.0591
          interpMode=RCIM_Cubic
          leaveTangent=-0.9637
          leaveTangentWeight=1.1238
          tangentMode=RCTM_User
          time=0.025
          value=1
         }
         {
          arriveTangent=-0.002
          arriveTangentWeight=0.6027
          interpMode=RCIM_Cubic
          leaveTangent=-0.002
          leaveTangentWeight=0.6027
          tangentMode=RCTM_User
          tangentWeightMode=RCTWM_WeightedBoth
          time=1
          value=0
         }
        ]}}}
     Scale Curve=&Multiply_Float(
      A=&FNTimeOfDayScalar(Night Scale=0.7, Morning Scale=0.8, Evening Scale=0.8)
      B=&MakeCustomFloatFromBool(Bool=&SetBoolByFloatComparison(A=$System.Blink, B=0.4), True Float=$System.Blink, False Float=1)
     )
    )
    Scale RGB=&VectorFromFloat(Value=&FloatFromCurve(FloatCurve=DataInterface() {propertyValues {Curve {keys [
          {
           arriveTangent=9.0725
           interpMode=RCIM_Cubic
           leaveTangent=9.0725
           leaveTangentWeight=0.1521
           tangentMode=RCTM_User
           tangentWeightMode=RCTWM_WeightedBoth
           time=0
           value=1
          }
          {
           arriveTangent=-5.1735
           arriveTangentWeight=0.0591
           interpMode=RCIM_Cubic
           leaveTangent=-5.1735
           leaveTangentWeight=1.1238
           tangentMode=RCTM_User
           time=0.025
           value=4
          }
          {
           arriveTangent=-0.002
           arriveTangentWeight=0.6027
           interpMode=RCIM_Cubic
           leaveTangent=-0.002
           leaveTangentWeight=0.6027
           tangentMode=RCTM_User
           tangentWeightMode=RCTWM_WeightedBoth
           time=1
           value=1
          }
         ]}}}))
    ScaleA=true
    ScaleRGB=true
   )
   &ScaleSpriteSize(
    Non-Uniform Curve Sprite Scale=DataInterface() {propertyValues {
      XCurve {keys [<at 0 1>, {arriveTangent=1.703, interpMode=RCIM_Cubic, leaveTangent=1.703, tangentMode=RCTM_User, time=1, value=2}]}
      YCurve {keys [
        {arriveTangent=9.0725, interpMode=RCIM_Cubic, leaveTangent=9.0725, tangentMode=RCTM_User, time=0, value=0.5}
        {arriveTangent=-0.7034, interpMode=RCIM_Cubic, leaveTangent=-0.7034, tangentMode=RCTM_User, time=0.1, value=1.25}
        <at 1 0.75>
       ]}
     }}
    Scale Sprite Size Mode=::NewEnumerator3
   )
  ]
  rendererClasses [&NiagaraSpriteRendererProperties]
  renderers [{rendererClass=&NiagaraSpriteRendererProperties}]
 }
 Glow_Eyes003 {
  emitterUpdateScript [
   &EmitterState()
   &SpawnBurst_Instantaneous(Spawn Count=2, Use Spawn Probability=false, UseLoopCountLimit=false)
  ]
  particleSpawnScript [
   &PartitionParticles()
   &InitializeParticle(
    Color=<rgba 6 0 0 1>
    Lifetime=1.2
    Sprite Size Mode=::NewEnumerator3
    Uniform Sprite Size=6.5
    UsePositionOffset=false
    Write Material Random=true
   )
   &CameraOffset(Camera Offset Amount=3)
  ]
  particleUpdateScript [
   &ParticleState()
   &SkeletalMeshLocation(
    Bone / Socket Index=$Particles.PartitionParticles.Partition
    Bone Sampling Mode=::NewEnumerator4
    MaskSpawnGroup=false
    Sampled Position Offset=&Lerp_Vector(
     A=&Lerp_Vector(
      A=<0.275 0.2 1.5>
      B=<-0.275 0.2 1.5>
      Alpha=&MakeFloatFromInt(Int=$Particles.PartitionParticles.Partition)
     )
     B=&Lerp_Vector(
      A=<1.1 1.25 1.25>
      B=<-1.1 1.25 1.25>
      Alpha=&MakeFloatFromInt(Int=$Particles.PartitionParticles.Partition)
     )
     Alpha=$System.Style
    )
    Skeletal Mesh=$User.SkeletalMesh
   )
   &ScaleColor(
    Scale Alpha=&FloatFromCurve(
     FloatCurve=DataInterface() {propertyValues {Curve {keys [
         {
          arriveTangent=9.0725
          interpMode=RCIM_Cubic
          leaveTangent=9.0725
          leaveTangentWeight=0.1521
          tangentMode=RCTM_User
          tangentWeightMode=RCTWM_WeightedBoth
          time=0
          value=0
         }
         {
          arriveTangent=-0.9637
          arriveTangentWeight=0.0591
          interpMode=RCIM_Cubic
          leaveTangent=-0.9637
          leaveTangentWeight=1.1238
          tangentMode=RCTM_User
          time=0.025
          value=1
         }
         {
          arriveTangent=-0.002
          arriveTangentWeight=0.6027
          interpMode=RCIM_Cubic
          leaveTangent=-0.002
          leaveTangentWeight=0.6027
          tangentMode=RCTM_User
          tangentWeightMode=RCTWM_WeightedBoth
          time=1
          value=0
         }
        ]}}}
     Scale Curve=&Multiply_Float(
      A=&FNTimeOfDayScalar(Night Scale=0.7, Morning Scale=0.8, Evening Scale=0.8)
      B=&MakeCustomFloatFromBool(Bool=&SetBoolByFloatComparison(A=$System.Blink, B=0.4), True Float=$System.Blink, False Float=1)
     )
    )
    Scale RGB=&VectorFromFloat(Value=&FloatFromCurve(FloatCurve=DataInterface() {propertyValues {Curve {keys [
          {
           arriveTangent=9.0725
           interpMode=RCIM_Cubic
           leaveTangent=9.0725
           leaveTangentWeight=0.1521
           tangentMode=RCTM_User
           tangentWeightMode=RCTWM_WeightedBoth
           time=0
           value=1
          }
          {
           arriveTangent=-5.1735
           arriveTangentWeight=0.0591
           interpMode=RCIM_Cubic
           leaveTangent=-5.1735
           leaveTangentWeight=1.1238
           tangentMode=RCTM_User
           time=0.025
           value=4
          }
          {
           arriveTangent=-0.002
           arriveTangentWeight=0.6027
           interpMode=RCIM_Cubic
           leaveTangent=-0.002
           leaveTangentWeight=0.6027
           tangentMode=RCTM_User
           tangentWeightMode=RCTWM_WeightedBoth
           time=1
           value=1
          }
         ]}}}))
    ScaleA=true
    ScaleRGB=true
   )
  ]
  rendererClasses [&NiagaraSpriteRendererProperties]
  renderers [{rendererClass=&NiagaraSpriteRendererProperties}]
 }
}
summary {
 systemName=NS_Emote_LazarusLens_Eyes_Reverse
 emitters [
  {emitterName=Glow_Eyes002, rendererClasses [&NiagaraSpriteRendererProperties]}
  {emitterName=Glow_Eyes003, rendererClasses [&NiagaraSpriteRendererProperties]}
  {bEnabled=false, emitterName=Glow_Eyes, rendererClasses [&NiagaraSpriteRendererProperties]}
  {bEnabled=false, emitterName=Glow_Eyes001, rendererClasses [&NiagaraSpriteRendererProperties]}
 ]
 userVariables [
  User.bIsFrontEnd {type=NiagaraBool}
  User.bIsFrontEndPreview {type=NiagaraBool}
  User.Kado {type=NiagaraFloat, defaultValue=0, description=""}
  User.Mask {type=NiagaraFloat, defaultValue=0, description=""}
  User.SkeletalMesh {
   type=NiagaraDataInterfaceSkeletalMesh
   defaultValue {
    struct=NiagaraExt_VariableValue_DataInterface
    value {dataInterface=&self:NiagaraDataInterfaceSkeletalMesh_0}
   }
  }
 ]
}
